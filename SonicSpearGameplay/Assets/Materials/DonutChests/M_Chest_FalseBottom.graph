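# M_Chest_FalseBottom is a Default Lit material, named and foldered as the false bottom of a donut chest.
# Its graph holds a single golden-orange constant color, with no textures, parameters or math.
# No link from that color to a material output appears in the graph.
# Automatic usage flag setting in the editor is turned off.
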Material /SonicSpearGameplay/Assets/Materials/DonutChests/M_Chest_FalseBottom
graph {
 properties {bAutomaticallySetUsageInEditor=false, usedShadingModels=Default Lit}
 nodes [Constant3Vector() {constant=<rgba 0.6584 0.3372 0.0296 1>}]
}
