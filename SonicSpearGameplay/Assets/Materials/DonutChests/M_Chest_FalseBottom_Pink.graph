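# Pink material for the false bottom of donut chests, going by its name and folder, using the Default Lit shading model.
# The graph holds a single constant color node in a soft lilac-pink, with no textures, parameters or math, and no connection to a material output shown.
# Automatic setting of usage flags in the editor is turned off.
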Material /SonicSpearGameplay/Assets/Materials/DonutChests/M_Chest_FalseBottom_Pink
graph {
 properties {bAutomaticallySetUsageInEditor=false, usedShadingModels=Default Lit}
 nodes [Constant3Vector() {constant=<rgba 0.521 0.305 0.6654 1>}]
}
