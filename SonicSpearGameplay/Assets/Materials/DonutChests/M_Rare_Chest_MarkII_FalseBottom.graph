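# Default Lit material, named for the false bottom of the Mark II rare chest in the DonutChests set.
# Its graph holds a single constant blue color, with no textures, parameters or math, and no link from that color to a material output is shown.
# Automatic setting of usage flags in the editor is turned off.
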Material /SonicSpearGameplay/Assets/Materials/DonutChests/M_Rare_Chest_MarkII_FalseBottom
graph {
 properties {bAutomaticallySetUsageInEditor=false, usedShadingModels=Default Lit}
 nodes [Constant3Vector() {constant=<rgba 0.0529 0.1413 0.4233 1>}]
}
